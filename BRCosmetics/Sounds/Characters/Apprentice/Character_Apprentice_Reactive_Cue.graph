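# Sound cue for the Apprentice character cosmetic that plays one of five kill-reaction sounds.
# An equally weighted random node picks the wave and feeds a modulator that slightly varies volume while keeping pitch fixed.
# It uses basic Foley attenuation and the gadgets sound class, with a concurrency override allowing one instance and stopping the oldest.
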SoundCue /BRCosmetics/Sounds/Characters/Apprentice/Character_Apprentice_Reactive_Cue
root=Modulator_0
nodes [
 {id=Modulator_0, children [Random_1], properties {pitchMax=1, pitchMin=1, volumeMax=1.1, volumeMin=1}}
 {id=Random_1, children [WavePlayer_3,WavePlayer_4,WavePlayer_5,WavePlayer_6,WavePlayer_7], properties {weights [1,1,1,1,1]}}
 {id=WavePlayer_3, properties {soundWaveAssetPtr=&/KillReaction_Apprentice_05}}
 {id=WavePlayer_4, properties {soundWaveAssetPtr=&/KillReaction_Apprentice_01}}
 {id=WavePlayer_5, properties {soundWaveAssetPtr=&/KillReaction_Apprentice_02}}
 {id=WavePlayer_6, properties {soundWaveAssetPtr=&/KillReaction_Apprentice_03}}
 {id=WavePlayer_7, properties {soundWaveAssetPtr=&/KillReaction_Apprentice_04}}
]
properties {
 attenuationSettings=&Foley_Attenuation_Basic_Att
 bOverrideConcurrency=true
 maxDistance=1536
 soundClassObject=&Fort_Gadgets
 volumeMultiplier=1
 concurrencyOverrides {maxCount=1, resolutionRule=StopOldest}
}
